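# Lit material for burning logs in the Princess Castle set, usable on mesh particles, that makes charcoal glow with animated embers.
# A boosted charcoal texture channel masks two panning noise layers, tinted by EmberColor and passed twice through MF_FX_EmissiveAdjustments.
# Object world X position divided by Offset shifts the noise UVs, so logs in different places vary.
# A darkened copy of the charcoal texture is also computed.
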Material /Game/Environments/Sets/PrincessCastle/FX/MT_BurningLog_Princess
graph {
 properties {bUsedWithMeshParticles=true, pixelDepthOffsetMode=PDOM_Legacy, usedShadingModels=Default Lit}
 nodes [
  Constant()
  Constant() {r=1}
  n1 = TextureSample(UVs=TextureCoordinate() {uTiling=2, vTiling=2}) {texture=/Game/Packages/Fortress_ShaderTests/Textures/charcoal}
  n2 = Power(Base=Multiply(A=@n1.G) {constB=4})
  n3 = ObjectPositionWS()
  n4 = ScalarParameter() {defaultValue=128, parameterName=Offset}
  n5 = Add(A=TextureCoordinate(), B=ComponentMask(None=Divide(A=@n3, B=@n4)) {r=true})
  n6 = VectorParameter() {parameterName=NoiseTile, defaultValue=<rgba 3 3 6 6>}
  n7 = ComponentMask(None=@n6.RGB) {g=true, r=true}
  n8 = ScalarParameter() {defaultValue=1.5, parameterName=MaskSpeed}
  n9 = Constant()
  n10 = TextureSample(UVs=Panner(Coordinate=Multiply(A=@n5, B=@n7), Speed=AppendVector(A=@n8, B=@n9))) {samplerType=SAMPLERTYPE_Grayscale, texture=/Game/Global/Textures/Default/Noise/TilingNoise06}
  n11 = ScalarParameter() {defaultValue=2, parameterName=MaskExp}
  n12 = Multiply(A=@n5, B=AppendVector(A=@n6.B, B=@n6.B))
  n13 = AppendVector(A=ScalarParameter() {defaultValue=2, parameterName=NoiseSpeed}, B=Constant())
  n14 = Multiply(
   A=Multiply(A=Saturate(None=@n2), B=Power(Base=@n10.G, Exp=@n11))
   B=TextureSample(UVs=Panner(Coordinate=@n12, Speed=@n13)) {texture=/Game/Global/Textures/Default/Noise/TankMoteNoise}.G
  )
  n15 = VectorParameter() {parameterName=EmberColor, defaultValue=<rgba 50 5 0 0>}
  MaterialFunctionCall(Emissive=MaterialFunctionCall(Emissive=Multiply(A=@n14, B=@n15.RGB)) {materialFunction=&MF_FX_EmissiveAdjustments}.Result) {materialFunction=&MF_FX_EmissiveAdjustments}
  Multiply(A=@n1.RGB) {constB=0.12}
  TextureCoordinate()
  WorldPosition()
 ]
}
