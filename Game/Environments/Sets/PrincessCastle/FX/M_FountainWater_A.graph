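# Masked, dithered, lit water material for the Princess Castle fountain falls, with scrolling foam, detail color and flattened normals.
# Panning distortion noise offsets two foam and normal layers plus a water detail pattern; mobile paths skip that distortion.
# Foam, noise and UV gradient masks feed MF_ClipAlpha for opacity, with a fade along V.
# A hard step on UV1's V blends between "Top" and regular parameter sets, giving two regions separate looks in one material.
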Material /Game/Environments/Sets/PrincessCastle/FX/M_FountainWater_A
graph {
 properties {
  blendMode=BLEND_Masked
  ditherOpacityMask=true
  numCustomizedUVs=1
  opacityMaskClipValue=0.999
  pixelDepthOffsetMode=PDOM_Legacy
  previewMesh=&/S_PrincessCastle_FountainWater_Falls
  translucencyLightingMode=TLM_VolumetricPerVertexNonDirectional
  translucentShadowDensityScale=0
  usedShadingModels=Default Lit
 }
 nodes [
  n1 = TextureCoordinate()
  n2 = ComponentMask(None=VectorParameter() {parameterName=DarkNoiseTile, defaultValue=<rgba 2 0.6 0 0>}.RGB) {g=true, r=true}
  n3 = Constant()
  n4 = ScalarParameter() {defaultValue=-0.6, parameterName=Panner_A_Speed}
  n5 = TextureSampleParameter2D(UVs=Panner(Coordinate=Multiply(A=@n1, B=@n2), Speed=AppendVector(A=@n3, B=@n4)) {speedY=1}) {parameterName=DistNoise, texture=/Game/Global/Textures/Default/Tileable_Patterns/T_Tilling_Distortion_Map_01}
  n6 = Multiply(A=ComponentMask(None=@n5.G) {g=true, r=true}, B=ScalarParameter() {defaultValue=0.35, parameterName=NoiseDistAmnt}) {constB=0.6}
  n7 = VectorParameter() {parameterName=FoamNoiseTile, defaultValue=<rgba 2 0.5 2 2>}
  n8 = Multiply(A=TextureCoordinate(), B=ComponentMask(None=@n7.RGB) {g=true, r=true})
  n9 = VectorParameter() {parameterName=NoisePanner, defaultValue=<rgba -0.1 -2 0.2 -3>}
  n10 = Panner(
   Coordinate=ShadingPathSwitch(Default=Add(A=@n6, B=@n8), Mobile=@n8) {desc=Migrated from FeatureLevelSwitch}
   Speed=ComponentMask(None=@n9.RGB) {g=true, r=true}
  ) {speedX=-0.5, speedY=1.6}
  n11 = TextureSample(UVs=@n10) {samplerType=SAMPLERTYPE_Normal, texture=/Game/Packages/Fortress_Transylvania/Haunted_Forest/Swamps/Textures/T_sea_foam_03_n}
  n12 = Multiply(A=AppendVector(A=@n7.B, B=@n7.B), B=TextureCoordinate())
  n13 = Panner(
   Coordinate=ShadingPathSwitch(Default=Add(A=@n6, B=@n12), Mobile=@n12) {desc=Migrated from FeatureLevelSwitch}
   Speed=AppendVector(A=@n9.B, B=@n9.B)
  ) {speedX=0.4, speedY=1.2}
  n14 = TextureSample(UVs=@n13) {samplerType=SAMPLERTYPE_Normal, texture=&T_sea_foam_03_n}
  n15 = ComponentMask(None=TextureCoordinate() {coordinateIndex=1}) {g=true}
  n16 = Saturate(None=Ceil(None=Subtract(A=@n15) {constB=0.5}))
  MaterialFunctionCall(
   Normal=ShadingPathSwitch(Default=Add(A=@n11.RGB, B=@n14.RGB), Mobile=@n11.RGB) {desc=Migrated from FeatureLevelSwitch}
   Flatness=LinearInterpolate(
    A=ScalarParameter() {defaultValue=0.5, parameterName=NormalFlatnessTop}
    B=ScalarParameter() {defaultValue=0.8, parameterName=NormalFlatness}
    Alpha=@n16
   )
  ) {materialFunction=&FlattenNormal}
  n17 = Add(
   A=TextureSampleParameter2D(UVs=@n10) {parameterName=FoamNoise, texture=/Game/Packages/Fortress_Transylvania/Haunted_Forest/Swamps/Textures/T_sea_foam_02}.G
   B=TextureSampleParameter2D(UVs=@n13) {parameterName=FoamNoise, texture=&T_sea_foam_02}.G
  )
  n18 = ComponentMask(None=TextureCoordinate() {coordinateIndex=1}) {g=true}
  n19 = Saturate(None=Ceil(None=Subtract(A=@n18) {constB=0.5}))
  n20 = Add(
   A=@n5.G
   B=LinearInterpolate(A=ScalarParameter() {parameterName=InnerNoiseAddTop}, B=ScalarParameter() {defaultValue=0.25, parameterName=InnerNoiseAdd}, Alpha=@n19)
  )
  n21 = Constant() {r=1}
  n22 = MaterialFunctionCall() {materialFunction=&LinearGradient}
  n23 = OneMinus(None=@n22.VGradient)
  n24 = Multiply(A=@n23) {constB=2}
  n25 = Power(
   Base=@n23
   Exp=LinearInterpolate(A=ScalarParameter() {defaultValue=1, parameterName=CenterExpTop}, B=ScalarParameter() {parameterName=CenterExp}, Alpha=@n19)
  ) {constExponent=20}
  n26 = Add(
   A=Multiply(A=ShadingPathSwitch(Default=@n20, Mobile=@n21) {desc=Migrated from FeatureLevelSwitch}, B=Saturate(None=@n24))
   B=Saturate(None=Multiply(A=@n25) {constB=2})
  )
  n27 = LinearInterpolate(
   A=ScalarParameter() {defaultValue=0.5, parameterName=BottomExpTop}
   B=ScalarParameter() {defaultValue=0.5, parameterName=BottomExp}
   Alpha=@n19
  )
  n28 = Saturate(None=Multiply(A=Power(Base=@n23, Exp=@n27) {constExponent=3}) {constB=2})
  n29 = ScalarParameter() {defaultValue=3, parameterName=SideMult}
  n30 = ScalarParameter() {defaultValue=5, parameterName=SideExp}
  n31 = Power(Base=Multiply(A=OneMinus(None=@n22.UGradient), B=@n29), Exp=@n30)
  n32 = Power(Base=Multiply(A=@n22.UGradient, B=@n29), Exp=@n30)
  n33 = LinearInterpolate(
   B=Saturate(None=Multiply(A=@n31, B=@n32))
   Alpha=LinearInterpolate(A=ScalarParameter() {parameterName=SideMaskAmntTop}, B=ScalarParameter() {parameterName=SideMaskAmnt}, Alpha=@n19)
  ) {constA=1}
  n34 = Saturate(None=Multiply(A=Add(A=@n17, B=@n26), B=Multiply(A=@n28, B=@n33)))
  n35 = ScalarParameter() {defaultValue=0.1, parameterName=ClipRange}
  n36 = LinearInterpolate(
   A=ScalarParameter() {defaultValue=0.5, parameterName=ClipTestTop}
   B=ScalarParameter() {defaultValue=0.7, parameterName=ClipTest}
   Alpha=@n19
  )
  n37 = OneMinus(None=MaterialFunctionCall() {materialFunction=&LinearGradient}.VGradient)
  n38 = ScalarParameter() {defaultValue=10, parameterName=TopFade}
  Multiply(
   A=Add(
    A=MaterialFunctionCall!(AlphaTexture=@n34, ClipRange=@n35, AlphaOverTime=@n36) {materialFunction=/Game/Effects/Fort_Effects/Materials/_Functions/MF_ClipAlpha}.Result
    B=ScalarParameter() {defaultValue=0.6, parameterName=BaseOpac}
   )
   B=OneMinus(None=Power(Base=@n37, Exp=@n38) {constExponent=20})
  )
  n39 = VectorParameter() {parameterName=DetailTile, defaultValue=<rgba 2 0.1 0 0>}
  n40 = Panner(
   Coordinate=Multiply(A=TextureCoordinate() {vTiling=2}, B=ComponentMask(None=@n39.RGB) {g=true, r=true})
   Speed=ComponentMask(None=VectorParameter() {parameterName=DetailPan, defaultValue=<rgba 0 -2 0 0>}.RGB) {g=true, r=true}
  )
  n41 = TextureSampleParameter2D(UVs=ShadingPathSwitch(Default=Add(A=@n40, B=@n6), Mobile=@n40) {desc=Migrated from FeatureLevelSwitch}) {parameterName=Detail, texture=/Game/Effects/Fort_Effects/Textures/Water/T_TilingWater_Pattern}
  n42 = ScalarParameter() {defaultValue=0.6, parameterName=DetailMult}
  n43 = Saturate(None=Add(A=Multiply(A=@n41.G, B=@n42), B=@n17))
  n44 = Multiply(
   A=VectorParameter() {parameterName=Color, defaultValue=<rgba 0 0.0671 0.1042 1>}.RGB
   B=LinearInterpolate(A=ScalarParameter() {defaultValue=0.7, parameterName=Dark}, Alpha=@n43)
  )
  n45 = ScalarParameter() {defaultValue=1, parameterName=FoamBrightTop}
  n46 = ScalarParameter() {defaultValue=0.7, parameterName=FoamBright}
  n47 = ComponentMask(None=TextureCoordinate() {coordinateIndex=1}) {g=true}
  n48 = Saturate(None=Ceil(None=Subtract(A=@n47) {constB=0.5}))
  n49 = Power(
   Base=@n23
   Exp=LinearInterpolate(A=ScalarParameter() {defaultValue=1, parameterName=FoamExpTop}, B=ScalarParameter() {defaultValue=0.6, parameterName=FoamExp}, Alpha=@n16)
  ) {constExponent=3}
  n50 = LinearInterpolate(A=Multiply(A=@n43, B=LinearInterpolate(A=@n45, B=@n46, Alpha=@n48)) {constB=0.2}, Alpha=Saturate(None=Multiply(A=@n49) {constB=2})) {constB=0}
  Multiply(
   A=MaterialFunctionCall(DiffuseColor=Add(A=@n44, B=@n50)) {materialFunction=&ConvertFromDiffSpec}.BaseColor
   B=ScalarParameter() {parameterName=EmissiveMult}
  )
  ScalarParameter() {defaultValue=1, parameterName=Spec}
  ScalarParameter() {defaultValue=0.3, parameterName=Roughness}
 ]
}
