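# Translucent water material for the Princess Castle fountain pools, with animated ripples, foam and reflections.
# Radial-remapped UVs drive opposite-panning normal noise, blended and flattened by radial gradients, which perturbs a time-of-day-tinted cubemap reflection and a Fresnel-style highlight.
# Radially masked foam and depth-fade edge glow brighten the base color, while another depth fade softens opacity at intersections.
# Panned distortion noise, weighted by UV1's vertical coordinate, forms a Z-axis vertex displacement.
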Material /Game/Environments/Sets/PrincessCastle/FX/M_Fountain_Pools
graph {
 properties {
  blendMode=BLEND_Translucent
  pixelDepthOffsetMode=PDOM_Legacy
  previewMesh=&/S_PrincessCastle_Fountain_pools
  usedShadingModels=Default Lit
 }
 nodes [
  n1 = MaterialFunctionCall() {materialFunction=&Color_Time_of_Day}
  n2 = TextureCoordinate()
  n3 = Constant() {r=0.5}
  n4 = Multiply(
   A=MaterialFunctionCall(Vector or UVs=Subtract(A=@n2, B=@n3)) {materialFunction=&VectorToRadialValue}.Radial Coordinates
   B=ComponentMask(None=VectorParameter() {parameterName=Tile, defaultValue=<rgba 1 2 0 0>}.RGB) {g=true, r=true}
  )
  n5 = VectorParameter() {parameterName=NoisePanner, defaultValue=<rgba 0 -0.25 0.2 -3>}
  n6 = TextureSampleParameter2D!(UVs=Panner(Coordinate=@n4, Speed=@n5.RGB) {speedX=-0.5, speedY=1.6}, Level, Apply View MipBias) {
   mipValueMode=TMVM_MipLevel
   parameterName=NormalNoise
   samplerType=SAMPLERTYPE_Normal
   texture=/Game/Packages/Fortress_Transylvania/Haunted_Forest/Swamps/Textures/T_sea_foam_02_n
  }
  n7 = ScalarParameter() {defaultValue=1, parameterName=NorMult}
  n8 = Saturate(None=MaterialFunctionCall(
    UVs=TextureCoordinate()
    CenterPosition=Constant2Vector() {g=0.5, r=0.5}
    Radius=ScalarParameter() {defaultValue=1, parameterName=NormRadiusInner}
    Density=ScalarParameter() {defaultValue=1, parameterName=NormDensityInner}
   ) {materialFunction=&RadialGradientExponential}.RadialGradientExponential)
  n9 = Panner(Coordinate=@n4, Speed=Multiply(A=@n5.RGB) {constB=-1}) {speedX=-0.5, speedY=1.6}
  n10 = TextureSampleParameter2D(UVs=@n9) {
   mipValueMode=TMVM_MipLevel
   parameterName=NormalNoise
   samplerType=SAMPLERTYPE_Normal
   texture=&T_sea_foam_02_n
  }
  n11 = MaterialFunctionCall(
   UVs=TextureCoordinate()
   CenterPosition=Constant2Vector() {g=0.5, r=0.5}
   Radius=ScalarParameter() {defaultValue=1, parameterName=NormRadius}
   Density=ScalarParameter() {defaultValue=1, parameterName=NormDesnity}
  ) {materialFunction=&RadialGradientExponential}
  n12 = LinearInterpolate(
   A=LinearInterpolate(A=Multiply(A=@n6.RGB, B=@n7), B=Constant3Vector() {constant=<rgba 0 0 1 0>}, Alpha=OneMinus(None=@n8))
   B=Multiply(A=@n7, B=@n10.RGB)
   Alpha=OneMinus(None=Saturate(None=@n11.RadialGradientExponential))
  )
  n13 = TextureSampleParameterCube!(UVs=Add(A=ReflectionVectorWS(), B=@n12), Level, Apply View MipBias) {
   mipValueMode=TMVM_MipLevel
   parameterName=Cube
   samplerType=SAMPLERTYPE_LinearColor
   texture=/Game/Weapons/WeaponSkins/Textures/AthenaCubemap2
  }
  n14 = Multiply(A=@n13.RGB, B=ScalarParameter() {defaultValue=2, parameterName=CubeBrightness})
  n15 = Saturate(None=DotProduct(A=@n12, B=CameraVectorWS()))
  n16 = Multiply(
   A=Power(Base=OneMinus(None=@n15), Exp=ScalarParameter() {defaultValue=40, parameterName=SpecExp})
   B=ScalarParameter() {defaultValue=0.95, parameterName=SpecMult}
  )
  n17 = VectorParameter() {parameterName=Color, defaultValue=<rgba 0.0623 0.3827 0.7031 0>}
  n18 = TextureSampleParameter2D(UVs=@n9) {
   mipValueMode=TMVM_MipLevel
   parameterName=FoamNoise
   texture=/Game/Packages/Fortress_Transylvania/Haunted_Forest/Swamps/Textures/T_sea_foam_02
  }
  n19 = Saturate(None=MaterialFunctionCall(
    UVs=TextureCoordinate()
    CenterPosition=Constant2Vector() {g=0.5, r=0.5}
    Radius=ScalarParameter() {defaultValue=1, parameterName=RGBRadius}
    Density=ScalarParameter() {defaultValue=1, parameterName=RGBDesnity}
   ) {materialFunction=&RadialGradientExponential}.RadialGradientExponential)
  n20 = DepthFade(Opacity=Constant() {r=1}, FadeDistance=ScalarParameter() {defaultValue=15, parameterName=BrightEdgeDist})
  n21 = Add(
   A=Multiply(A=@n18.G, B=OneMinus(None=@n19))
   B=Multiply(A=OneMinus(None=@n20), B=ScalarParameter() {defaultValue=3, parameterName=EdgeBrightness})
  )
  Add(A=Add(A=Multiply(A=@n1.Result, B=@n14), B=Saturate(None=@n16)), B=Add(A=@n17.RGB, B=Multiply(A=@n17.RGB, B=@n21)))
  DepthFade(
   Opacity=ScalarParameter() {defaultValue=1, parameterName=Opacity}
   FadeDistance=ScalarParameter() {defaultValue=15, parameterName=EdgeFadeDist}
  )
  n22 = Constant()
  n23 = ScalarParameter() {defaultValue=1, parameterName=WPO_Speed}
  n24 = TextureSampleParameter2D!(UVs=Panner(Coordinate=@n4, Speed=AppendVector(A=@n22, B=@n23)), Apply View MipBias) {parameterName=WPONoise, texture=/Game/Global/Textures/Default/Tileable_Patterns/T_Tilling_Distortion_Map_01}
  n25 = ScalarParameter() {defaultValue=20, parameterName=WPOAmnt}
  n26 = TextureCoordinate() {coordinateIndex=1}
  Multiply(
   A=Multiply(A=Multiply(A=@n24.G, B=@n25), B=Constant3Vector() {constant=<rgba 0 0 1 0>})
   B=LinearInterpolate(Alpha=ComponentMask(None=@n26) {g=true}) {constA=0.25}
  )
 ]
}
