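# Placeholder-like material in the PrincessCastle environment set, using the Default Lit shading model.
# Its graph holds a single vector parameter with a dark grey default color and no textures or math.
# The file shows no link from that parameter to a material output.
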Material /Game/Environments/Sets/PrincessCastle/Materials/05___Default
graph {
 properties {pixelDepthOffsetMode=PDOM_Legacy, usedShadingModels=Default Lit}
 nodes [VectorParameter() {defaultValue=<rgba 0.1333 0.1333 0.1333 0>}]
}
